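# Post-process style AnimBlueprint for the GrumbleCroak head cosmetic, adding secondary motion to frills and tongue and posing lizard eyelids.
# Spring-driven AnimDynamics nodes simulate three frill bones and the tongue tip, while constraints and additive offsets shape the tongue and lip.
# Additive lid rotations precede translation constraints to the eyes.
# The event graph derives blink values from curves, an emote flag, and a simulation space switched for front-end or DBNO-carried states.
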AnimBlueprint /BRCosmetics/Characters/Player/Male/Medium/Heads/M_MED_GrumbleCroak_Head/Meshes/Fortnite_Base_Head_GrumbleCroak_Skeleton_AnimBP_CustomDyn
target_skeleton=&/Fortnite_Base_Head_GrumbleCroak_Skeleton
graphs [
 AnimGraph {nodes [
   n1 = Anim Dynamics - Bone: dyn_frillTop!(SimulationSpace:AnimPhysSimSpaceType Enum, ComponentPose=Local To Component(LocalPose=Input Pose()), Alpha=1, ExternalForce="0,0,0") -> (Pose) {
    angularDampingOverride=1
    angularSpringConstant=8000
    bAngularSpring=true
    bOverrideAngularDamping=true
    bUseGravityOverride=true
    gravityScale=0.35
    simulationSpace=RootRelative
    alphaBoolBlend {blendOption=Linear}
    boundBone {boneName=dyn_frillTop}
    componentAppliedLinearAccClamp=<2000 2000 2000>
    componentLinearAccScale=<1 1 1>
    componentLinearVelScale=<1 1 1>
    gravityOverride {z=-980}
    physicsBodyDefinitions [{
      boundBone {boneName=dyn_frillTop}
      boxExtents=<22 22 22>
      constraintSetup {angularLimitsMax=<15 20 15>, angularLimitsMin=<-15 -50 -15>, angularTarget=<1 0 0>}
      localJointOffset=<3 0 0>
     }]
   }
   n2 = Anim Dynamics - Bone: dyn_frillMid!(SimulationSpace:AnimPhysSimSpaceType Enum, ComponentPose=@n1, Alpha=1, ExternalForce="0,0,0") -> (Pose) {
    angularDampingOverride=1
    angularSpringConstant=10000
    bAngularSpring=true
    bOverrideAngularDamping=true
    bUseGravityOverride=true
    gravityScale=0.35
    simulationSpace=RootRelative
    alphaBoolBlend {blendOption=Linear}
    boundBone {boneName=dyn_frillMid}
    componentAppliedLinearAccClamp=<2000 2000 2000>
    componentLinearAccScale=<1 1 1>
    componentLinearVelScale=<1 1 1>
    gravityOverride {z=-980}
    physicsBodyDefinitions [{
      boundBone {boneName=dyn_frillMid}
      boxExtents=<18 18 18>
      constraintSetup {angularLimitsMax=<15 20 15>, angularLimitsMin=<-15 -50 -15>, angularTarget=<1 0 0>}
      localJointOffset=<3 0 0>
     }]
   }
   n3 = Anim Dynamics - Bone: dyn_frillBottom!(SimulationSpace:AnimPhysSimSpaceType Enum, ComponentPose=@n2, Alpha=1, ExternalForce="0,0,0") -> (Pose) {
    angularDampingOverride=1
    angularSpringConstant=12000
    bAngularSpring=true
    bOverrideAngularDamping=true
    bUseGravityOverride=true
    gravityScale=0.35
    simulationSpace=RootRelative
    alphaBoolBlend {blendOption=Linear}
    boundBone {boneName=dyn_frillBottom}
    componentAppliedLinearAccClamp=<2000 2000 2000>
    componentLinearAccScale=<1 1 1>
    componentLinearVelScale=<1 1 1>
    gravityOverride {z=-980}
    physicsBodyDefinitions [{
      boundBone {boneName=dyn_frillBottom}
      boxExtents=<16 16 16>
      constraintSetup {angularLimitsMax=<15 20 15>, angularLimitsMin=<-15 -50 -15>, angularTarget=<1 0 0>}
      localJointOffset=<3 0 0>
     }]
   }
   n4 = "Constraint - dyn_tongue_3 (PP)"!(ComponentPose=@n3, Alpha=1, ConstraintWeights_0=0.6, ConstraintWeights_1:float=0.4) -> (Pose) {
    alphaBoolBlend {blendOption=Linear}
    boneToModify {boneName=dyn_tongue_3}
    constraintSetup [{targetBone {boneName=R_lip_lower_outer}}, {targetBone {boneName=C_jaw}}]
    constraintWeights [1,1]
   }
   n5 = "Transform (Modify) Bone - Bone: R_lip_lower_outer"(ComponentPose=@n4, Alpha=1, Translation="-2,0,-0.4", Rotation="5,0,0", Scale="1,1,1") {
    rotationMode=BMM_Additive
    rotationSpace=BCS_BoneSpace
    translationMode=BMM_Additive
    translationSpace=BCS_BoneSpace
    alphaBoolBlend {blendOption=Linear}
    boneToModify {boneName=R_lip_lower_outer}
    rotation {pitch=5}
    translation {x=-2, z=-0.4}
   }
   n6 = "Transform (Modify) Bone - Bone: dyn_tongue_1"(ComponentPose=@n5, Alpha=1, Translation="-2.3,1.9,2.1", Rotation="0,38,0", Scale="1,1,1") {
    rotationMode=BMM_Additive
    rotationSpace=BCS_BoneSpace
    scaleMode=BMM_Additive
    scaleSpace=BCS_BoneSpace
    translationMode=BMM_Additive
    translationSpace=BCS_BoneSpace
    alphaBoolBlend {blendOption=Linear}
    boneToModify {boneName=dyn_tongue_1}
    rotation {yaw=38}
    translation=<-2.3 1.9 2.1>
   }
   n7 = "Transform (Modify) Bone - Bone: dyn_tongue_3"(ComponentPose=@n6, Alpha=1, Translation="-0.18,1.2,1.1", Rotation="-20,32,30", Scale="1,1,0.4") {
    rotationMode=BMM_Additive
    rotationSpace=BCS_BoneSpace
    scaleMode=BMM_Additive
    scaleSpace=BCS_BoneSpace
    translationMode=BMM_Additive
    translationSpace=BCS_BoneSpace
    alphaBoolBlend {blendOption=Linear}
    boneToModify {boneName=dyn_tongue_3}
    rotation=<rot -20 32 30>
    scale {z=0.4}
    translation=<-0.18 1.2 1.1>
   }
   n8 = "Transform (Modify) Bone - Bone: dyn_tongue_4"(ComponentPose=@n7, Alpha=1, Translation="0,0,0.3", Rotation="0,0,40", Scale="1,1,1.6") {
    rotationMode=BMM_Additive
    rotationSpace=BCS_BoneSpace
    scaleMode=BMM_Additive
    scaleSpace=BCS_BoneSpace
    translationMode=BMM_Additive
    translationSpace=BCS_BoneSpace
    alphaBoolBlend {blendOption=Linear}
    boneToModify {boneName=dyn_tongue_4}
    rotation {roll=40}
    scale {z=1.6}
    translation {z=0.3}
   }
   n9 = "Transform (Modify) Bone - Bone: dyn_tongue_5"(ComponentPose=@n8, Alpha=1, Translation="0,0,0", Rotation="0,0,8", Scale="1,1,1") {
    rotationMode=BMM_Additive
    rotationSpace=BCS_BoneSpace
    scaleMode=BMM_Additive
    scaleSpace=BCS_BoneSpace
    translationMode=BMM_Additive
    translationSpace=BCS_BoneSpace
    alphaBoolBlend {blendOption=Linear}
    boneToModify {boneName=dyn_tongue_5}
    rotation {roll=8}
   }
   n10 = Anim Dynamics - Bone: dyn_tongue_4!(SimulationSpace:AnimPhysSimSpaceType Enum, ComponentPose=@n9, Alpha=1, ExternalForce="0,0,0") -> (Pose) {
    angularDampingOverride=1
    angularSpringConstant=5000
    bAngularSpring=true
    bOverrideAngularDamping=true
    bUseGravityOverride=true
    gravityScale=0.35
    simulationSpace=RootRelative
    alphaBoolBlend {blendOption=Linear}
    boundBone {boneName=dyn_tongue_4}
    componentAppliedLinearAccClamp=<2000 2000 2000>
    componentLinearAccScale=<1 1 1>
    componentLinearVelScale=<1 1 1>
    gravityOverride {z=-980}
    physicsBodyDefinitions [{
      boundBone {boneName=dyn_tongue_4}
      boxExtents=<16 16 16>
      constraintSetup {angularLimitsMax=<0 0 25>, angularLimitsMin=<-40 0 -25>, angularTarget=<1 0 0>}
      localJointOffset=<0 3 0>
     }]
   }
   n11 = Anim Dynamics - Bone: dyn_tongue_5!(SimulationSpace:AnimPhysSimSpaceType Enum, ComponentPose=@n10, Alpha=1, ExternalForce="0,0,0") -> (Pose) {
    angularDampingOverride=1
    angularSpringConstant=5000
    bAngularSpring=true
    bOverrideAngularDamping=true
    bUseGravityOverride=true
    gravityScale=0.35
    simulationSpace=RootRelative
    alphaBoolBlend {blendOption=Linear}
    boundBone {boneName=dyn_tongue_5}
    componentAppliedLinearAccClamp=<2000 2000 2000>
    componentLinearAccScale=<1 1 1>
    componentLinearVelScale=<1 1 1>
    gravityOverride {z=-980}
    physicsBodyDefinitions [{
      boundBone {boneName=dyn_tongue_5}
      boxExtents=<16 16 16>
      constraintSetup {angularLimitsMax=<10 0 25>, angularLimitsMin=<-40 0 -25>, angularTarget=<1 0 0>}
      localJointOffset=<0 3 0>
     }]
   }
   n12 = "Transform (Modify) Bone - Bone: L_eye_lizardLid"!(ComponentPose=@n11, Alpha=1, Rotation="0,-140,0") -> (Pose) {
    rotationMode=BMM_Additive
    rotationSpace=BCS_BoneSpace
    alphaBoolBlend {blendOption=Linear}
    alphaScaleBiasClamp {bInterpResult=true, interpSpeedDecreasing=6, interpSpeedIncreasing=15}
    boneToModify {boneName=L_eye_lizardLid}
    rotation {yaw=-140}
   }
   n13 = "Transform (Modify) Bone - Bone: R_eye_lizardLid"!(ComponentPose=@n12, Alpha=1, Rotation="0,140,0") -> (Pose) {
    rotationMode=BMM_Additive
    rotationSpace=BCS_BoneSpace
    alphaBoolBlend {blendOption=Linear}
    alphaScaleBiasClamp {bInterpResult=true, interpSpeedDecreasing=6, interpSpeedIncreasing=15}
    boneToModify {boneName=R_eye_lizardLid}
    rotation {yaw=140}
   }
   n14 = "Constraint - L_eye_lizardLid (T)"(ComponentPose=@n13, Alpha=1, ConstraintWeights_0=1) {
    alphaBoolBlend {blendOption=Linear}
    boneToModify {boneName=L_eye_lizardLid}
    constraintSetup [{transformType=Translation, targetBone {boneName=L_eye}}]
    constraintWeights [1]
   }
   n15 = "Constraint - R_eye_lizardLid (T)"(ComponentPose=@n14, Alpha=1, ConstraintWeights_0=1) {
    alphaBoolBlend {blendOption=Linear}
    boneToModify {boneName=R_eye_lizardLid}
    constraintSetup [{transformType=Translation, targetBone {boneName=R_eye}}]
    constraintWeights [1]
   }
   Output Pose(Result=Component To Local(ComponentPose=@n15))
  ]}
 EventGraph {nodes [
   Event BlueprintUpdateAnimation()
   > Set notPlayingEmote(notPlayingEmote:bool=NOT Boolean(A=Get bIsPlayingEmote())) -> (Output_Get:bool)
   > Set L_blinkPose(L_blinkPose:float=MapRangeClamped(Value=GetCurveValue(CurveName=L_blink_pose), InRangeA=0, InRangeB=2)) -> (Output_Get:float)
   > Set R_blinkPose(R_blinkPose:float=MapRangeClamped(Value=GetCurveValue(CurveName=R_blink_pose), InRangeA=0, InRangeB=2)) -> (Output_Get:float)
   > n1 = Branch(Condition=OR Boolean(A=Get bIsInFrontEnd(), B=Get bIsDBNOCarried()))
   > Set simSpaceChange(simSpaceChange:AnimPhysSimSpaceType Enum=World) -> (Output_Get:AnimPhysSimSpaceType Enum)
   Set simSpaceChange(execute=@n1.else)
  ]}
]
variables [
 L_blinkPose {type=number}
 notPlayingEmote {type=boolean}
 R_blinkPose {type=number}
 simSpaceChange {type=AnimPhysSimSpaceType}
]
