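# Animation Blueprint for the GrumbleCroak medium male head cosmetic, targeting the Fortnite_Base_Head_GrumbleCroak skeleton.
# It is a stub as shown: the event graph is empty, no variables are declared, and no animation graph logic appears.
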AnimBlueprint /BRCosmetics/Characters/Player/Male/Medium/Heads/M_MED_GrumbleCroak_Head/Meshes/M_MED_GrumbleCroak_Head_AnimBP
target_skeleton=&/Fortnite_Base_Head_GrumbleCroak_Skeleton
graphs [EventGraph {links [], nodes []}]
variables []
